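# Physics asset for the FeralTrash backpack cosmetic (NOHB variant), defining secondary motion for its dangling trash: book, drips, fishbones, juice box, spray paint and ties.
# Small capsules sit on the pack and base_ anchor bones, with larger capsules on the dyn_ bones constrained to them.
# The book page, straw and three-link fishbones form chains, while each tie base carries two strands.
# Tight swing and near-locked twist limits keep movement subtle, with ties swinging widest.
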PhysicsAsset /BRCosmetics/Accessories/FORT_Backpacks/Backpack_M_MED_FeralTrash/Meshes/M_MED_FeralTrash_Pack_Physics_NOHB
bodies {
 base_book [{shapeType=Capsule, length=0.5, radius=0.5, shapeName=base_book_capsule}]
 base_drip_l [{shapeType=Capsule, length=0.5, radius=0.5, shapeName=base_drip_l_capsule}]
 base_drip_m [{shapeType=Capsule, length=0.5, radius=0.5, shapeName=base_drip_m_capsule}]
 base_drip_r [{shapeType=Capsule, length=0.5, radius=0.5, shapeName=base_drip_r_capsule}]
 base_fishbones [{shapeType=Capsule, length=0.5, radius=0.5, shapeName=base_fishbones_capsule}]
 base_juicebox [{shapeType=Capsule, length=0.5, radius=0.5, shapeName=base_juicebox_capsule}]
 base_spraypaint [{shapeType=Capsule, length=0.5, radius=0.5, shapeName=base_spraypaint_capsule}]
 base_tie_back [{shapeType=Capsule, length=0.5, radius=0.5, shapeName=base_tie_back_capsule}]
 base_tie_r [{shapeType=Capsule, length=0.5, radius=0.5, shapeName=base_tie_r_capsule}]
 dyn_book [{shapeType=Capsule, length=1, radius=1, shapeName=dyn_book_capsule, center=<1 0 0>, rotation=<rot 90 0 0>}]
 dyn_book_page [{shapeType=Capsule, length=1, radius=1, shapeName=dyn_book_page_capsule, rotation=<rot 90 0 0>}]
 dyn_drip_l [{shapeType=Capsule, length=1, radius=1, shapeName=dyn_drip_l_capsule, center=<2 0 0>, rotation=<rot 90 0 0>}]
 dyn_drip_m [{shapeType=Capsule, length=1, radius=1, shapeName=dyn_drip_m_capsule, center=<2 0 0>, rotation=<rot 90 0 0>}]
 dyn_drip_r [{shapeType=Capsule, length=1, radius=1, shapeName=dyn_drip_r_capsule, center=<2 0 0>, rotation=<rot 90 0 0>}]
 dyn_fishbones [{shapeType=Capsule, length=1, radius=1, shapeName=dyn_fishbones_capsule, center=<2 0 0>, rotation=<rot 90 0 0>}]
 dyn_fishbones1 [{shapeType=Capsule, length=1, radius=1, shapeName=dyn_fishbones1_capsule, center=<2 0 0>, rotation=<rot 90 0 0>}]
 dyn_fishbones_end [{shapeType=Capsule, length=1, radius=1, shapeName=dyn_fishbones_end_capsule, center=<2 0 0>, rotation=<rot 90 0 0>}]
 dyn_juicebox [{shapeType=Capsule, length=1, radius=1, shapeName=dyn_juicebox_capsule, rotation=<rot 90 0 0>}]
 dyn_spraypaint [{shapeType=Capsule, length=1, radius=1, shapeName=dyn_spraypaint_capsule, center=<1 0 0>, rotation=<rot 90 0 0>}]
 dyn_straw [{shapeType=Capsule, length=1, radius=1, shapeName=dyn_straw_capsule, rotation=<rot 90 0 0>}]
 dyn_tie_l_back [{shapeType=Capsule, length=1, radius=1, shapeName=dyn_tie_l_back_capsule, center=<1 0 0>, rotation=<rot 90 0 0>}]
 dyn_tie_l_r [{shapeType=Capsule, length=1, radius=1, shapeName=dyn_tie_l_r_capsule, center=<1 0 0>, rotation=<rot 90 0 0>}]
 dyn_tie_r_back [{shapeType=Capsule, length=1, radius=1, shapeName=dyn_tie_r_back_capsule, center=<1 0 0>, rotation=<rot 90 0 0>}]
 dyn_tie_r_r [{shapeType=Capsule, length=1, radius=1, shapeName=dyn_tie_r_r_capsule, center=<1 0 0>, rotation=<rot 90 0 0>}]
 pack [{shapeType=Capsule, length=0.5, radius=0.5, shapeName=pack_capsule}]
}
constraints [
 {bone1Name=dyn_book, bone2Name=base_book, swing1LimitDegrees=8, swing2LimitDegrees=8, twistLimitDegrees=1}
 {bone1Name=dyn_book_page, bone2Name=dyn_book, swing1LimitDegrees=5, swing2LimitDegrees=5, twistLimitDegrees=0}
 {bone1Name=dyn_drip_l, bone2Name=base_drip_l, swing1LimitDegrees=8, swing2LimitDegrees=8, twistLimitDegrees=3}
 {bone1Name=dyn_drip_m, bone2Name=base_drip_m, swing1LimitDegrees=8, swing2LimitDegrees=8, twistLimitDegrees=3}
 {bone1Name=dyn_drip_r, bone2Name=base_drip_r, swing1LimitDegrees=8, swing2LimitDegrees=8, twistLimitDegrees=3}
 {bone1Name=dyn_fishbones, bone2Name=base_fishbones, swing1LimitDegrees=10, swing2LimitDegrees=8, twistLimitDegrees=1}
 {bone1Name=dyn_fishbones1, bone2Name=dyn_fishbones, swing1LimitDegrees=10, swing2LimitDegrees=8, twistLimitDegrees=1}
 {bone1Name=dyn_fishbones_end, bone2Name=dyn_fishbones1, swing1LimitDegrees=10, swing2LimitDegrees=8, twistLimitDegrees=1}
 {bone1Name=dyn_juicebox, bone2Name=base_juicebox, swing1LimitDegrees=5, swing2LimitDegrees=5, twistLimitDegrees=1}
 {bone1Name=dyn_straw, bone2Name=dyn_juicebox, swing1LimitDegrees=10, swing2LimitDegrees=10, twistLimitDegrees=0}
 {bone1Name=dyn_spraypaint, bone2Name=base_spraypaint, swing1LimitDegrees=8, swing2LimitDegrees=8, twistLimitDegrees=1}
 {bone1Name=dyn_tie_l_back, bone2Name=base_tie_back, swing1LimitDegrees=15, swing2LimitDegrees=15, twistLimitDegrees=1}
 {bone1Name=dyn_tie_r_back, bone2Name=base_tie_back, swing1LimitDegrees=15, swing2LimitDegrees=15, twistLimitDegrees=1}
 {bone1Name=dyn_tie_l_r, bone2Name=base_tie_r, swing1LimitDegrees=15, swing2LimitDegrees=15, twistLimitDegrees=1}
 {bone1Name=dyn_tie_r_r, bone2Name=base_tie_r, swing1LimitDegrees=15, swing2LimitDegrees=15, twistLimitDegrees=1}
]
