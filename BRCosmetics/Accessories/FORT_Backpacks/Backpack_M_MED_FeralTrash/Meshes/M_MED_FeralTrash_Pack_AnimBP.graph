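# Animation Blueprint for the FeralTrash backpack cosmetic that adds physics-driven secondary motion to the pack.
# Two spring-driven Anim Dynamics bones handle the dyn_handsan chain, with an additive base_handsan rotation while skydiving, then cache the pose.
# The cache feeds two RigidBody sims, choosing the NOHB physics asset in the front end and the standard one otherwise.
# The update event refreshes those states and a state-based gravity override for the RigidBody nodes.
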AnimBlueprint /BRCosmetics/Accessories/FORT_Backpacks/Backpack_M_MED_FeralTrash/Meshes/M_MED_FeralTrash_Pack_AnimBP
target_skeleton=&/M_MED_FeralTrash_Pack_Skeleton
graphs [
 AnimGraph {nodes [
   n1 = RigidBody!(
    ComponentPose=Local To Component(LocalPose=Use cached pose 'Burger ON/OFF'())
    Alpha
    OverrideWorldGravity:Vector=Get GravityOverride()
    ExternalForce
   ) -> (Pose) {
    bClampLinearTranslationLimitToRefPose=true
    bOverrideWorldGravity=true
    overridePhysicsAsset=&/M_MED_FeralTrash_Pack_Physics_NOHB
    simulationSpace=BaseBoneSpace
    alphaBoolBlend {blendOption=Linear}
    baseBoneRef {boneName=root}
    componentAppliedLinearAccClamp=<3000 3000 3000>
    componentLinearAccScale=<0.2 0.2 0.2>
    componentLinearVelScale=<0.2 0.2 0.2>
    overrideWorldGravity {z=-980}
    simSpaceSettings {
     maxAngularAcceleration=5000
     maxAngularVelocity=5000
     maxLinearAcceleration=5000
     maxLinearVelocity=5000
     worldAlpha=0.1
     externalLinearDragV {y=50}
    }
   }
   n2 = RigidBody(ComponentPose=Local To Component(LocalPose=Use cached pose 'Burger ON/OFF'()), OverrideWorldGravity=Get GravityOverride()) {
    bClampLinearTranslationLimitToRefPose=true
    bOverrideWorldGravity=true
    overridePhysicsAsset=&/M_MED_FeralTrash_Pack_Physics
    simulationSpace=BaseBoneSpace
    alphaBoolBlend {blendOption=Linear}
    baseBoneRef {boneName=root}
    componentAppliedLinearAccClamp=<3000 3000 3000>
    componentLinearAccScale=<0.2 0.2 0.2>
    componentLinearVelScale=<0.2 0.2 0.2>
    overrideWorldGravity {z=-980}
    simSpaceSettings {
     maxAngularAcceleration=5000
     maxAngularVelocity=5000
     maxLinearAcceleration=5000
     maxLinearVelocity=5000
     worldAlpha=0.1
     externalLinearDragV {y=50}
    }
   }
   Output Pose(Result=Blend Poses by bool(BlendPose_0=Component To Local(ComponentPose=@n1), BlendPose_1=Component To Local(ComponentPose=@n2), bActiveValue=Get isFE() -> (isFE:bool)) {blendType=Linear}) {layerGroup=None}
   n3 = Anim Dynamics - Bone: dyn_handsan(ComponentPose=Local To Component(LocalPose=Input Pose()), Alpha=1, ExternalForce="0,0,0") {
    angularDampingOverride=0.98
    angularSpringConstant=10000
    bAngularSpring=true
    bOverrideAngularDamping=true
    bUseGravityOverride=true
    gravityScale=0.35
    alphaBoolBlend {blendOption=Linear}
    boundBone {boneName=dyn_handsan}
    componentLinearAccScale=<1 1 1>
    componentLinearVelScale=<1 1 1>
    gravityOverride {z=-1000}
    physicsBodyDefinitions [{
      boundBone {boneName=dyn_handsan}
      boxExtents=<20 30 30>
      constraintSetup {angularLimitsMax=<15 15 50>, angularLimitsMin=<-15 -15 0>, angularTarget=<1 0 0>}
      localJointOffset=<4 0 0>
     }]
   }
   n4 = Anim Dynamics - Bone: dyn_handsan_end(ComponentPose=@n3, Alpha=1, ExternalForce="0,0,0") {
    angularDampingOverride=0.98
    angularSpringConstant=10000
    bAngularSpring=true
    bOverrideAngularDamping=true
    bUseGravityOverride=true
    gravityScale=0.35
    alphaBoolBlend {blendOption=Linear}
    boundBone {boneName=dyn_handsan_end}
    componentLinearAccScale=<1 1 1>
    componentLinearVelScale=<1 1 1>
    gravityOverride {z=-3000}
    physicsBodyDefinitions [{
      boundBone {boneName=dyn_handsan_end}
      boxExtents=<20 30 30>
      constraintSetup {angularLimitsMax=<50 50 50>, angularLimitsMin=<-50 -50 -50>, angularTarget=<1 0 0>}
      localJointOffset=<4 0 0>
     }]
   }
   n5 = "Transform (Modify) Bone - Bone: base_handsan"(ComponentPose=@n4, Alpha=Get is skydiving() -> (is skydiving:double), Translation="0,0,0", Rotation="27,40.082,-36.576", Scale="1,1,1") {
    rotationMode=BMM_Additive
    rotationSpace=BCS_BoneSpace
    translationMode=BMM_Additive
    translationSpace=BCS_BoneSpace
    alphaBoolBlend {blendOption=Linear}
    alphaScaleBiasClamp {bInterpResult=true}
    boneToModify {boneName=base_handsan}
    rotation=<rot 27 40.082 -36.576>
   }
   Save cached pose 'Burger ON/OFF'(Pose=Component To Local(ComponentPose=@n5)) {cacheName=Burger ON/OFF}
  ]}
 EventGraph {nodes [
   Event BlueprintUpdateAnimation()
   > n1 = Gravity Override State Function DEPRECATED(CharacterPartAnimInstance=Self-Reference(), GameTimeInSeconds=GetGameTimeInSeconds(), GravityOverrideParams=Get Gravity Override Params())
   > Set GravityOverride(GravityOverride=@n1)
   > Set is skydiving(is skydiving:double="To Float (Boolean)"(InBool=Get bIsSkyDiving())) -> (Output_Get:double)
   n2 = NOT Boolean(A=Get bIsParachuteOpened())
   n3 = Get bIsSkyDiving()
   > Set skydiveNOparachute(skydiveNOparachute:double="To Float (Boolean)"(InBool=AND Boolean(A=@n3, B=@n2))) -> (Output_Get:double)
   > Set isFE(isFE:bool=Get bIsInFrontEnd()) -> (Output_Get:bool)
   TryGetPawnOwner()
  ]}
]
variables [
 Gravity Override Params {type=GravityOverrideParamsStruct, default {skydive_GoingUp {gravityOverride=<0 -300 -980>}}}
 GravityOverride {type=Vector}
 is skydiving {type=number}
 isFE {type=boolean}
 skydiveNOparachute {type=number}
]
